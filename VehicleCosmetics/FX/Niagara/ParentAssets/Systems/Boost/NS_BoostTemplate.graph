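# Parent Niagara template for vehicle boost effects, holding shared system-level setup and user parameters but no emitters.
# User parameters expose paint color and scalar, color variant, game mode, port count, and flags for painting, frontend preview and vehicle speed preview.
# The system pairs System State with three Set Variables modules, and uses skeletal mesh and float3 array data interfaces plus a dynamic input that selects a vector from an array.
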NiagaraSystem /VehicleCosmetics/FX/Niagara/ParentAssets/Systems/Boost/NS_BoostTemplate
dependencies {
 usedDataInterfaces [&NiagaraDataInterfaceSkeletalMesh, &NiagaraDataInterfaceArrayFloat3]
 usedDynamicInputs [&SelectVectorFromArray]
 usedModules [
  &self:SystemSpawnScript.SystemScriptSource.SystemScriptGraph.NiagaraNodeAssignment_0.SetVariables_#1
  &self:SystemSpawnScript.SystemScriptSource.SystemScriptGraph.NiagaraNodeAssignment_1.SetVariables_#2
  &self:SystemSpawnScript.SystemScriptSource.SystemScriptGraph.NiagaraNodeAssignment_2.SetVariables_#3
  &SystemState
 ]
}
summary {
 systemName=NS_BoostTemplate
 emitters []
 userVariables [
  User.bIsFrontendPreview {type=NiagaraBool}
  User.bIsPainted {type=NiagaraBool}
  User.blsIsCoroColor {type=NiagaraBool}
  User.bPreviewVehicleSpeed {type=NiagaraBool}
  User.ColorVariantID {type=NiagaraInt32}
  User.GameModeID {type=NiagaraInt32}
  User.PaintedColor {type=LinearColor, description="", defaultValue {struct=LinearColor, value=<rgba 1 1 1 1>}}
  User.PaintedScalar {type=NiagaraFloat, defaultValue=1, description=""}
  User.PortCount {type=NiagaraInt32, defaultValue=1}
 ]
}
